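# Material function that samples an extra color map for an impostor and outputs it as ColorMap.
# It unpacks the Impostor Data input, which uses material attributes as a generic struct carrying values such as UVs under unrelated labels.
# That data drives FrameBlendWeights_TextureObject to sample and blend the supplied texture across frames.
# The result feeds both the default and mobile inputs of OverrideMobile_V4 before output.
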MaterialFunction /ImpostorBaker/MaterialFunctions/Impostor_ExtraColorMap
graph {nodes [
  n1 = GetMaterialAttributes(None=FunctionInput() {bUsePreviewValueAsDefault=true, inputName=Impostor Data, inputType=FunctionInput_MaterialAttributes}) {
   desc="Note: Attributes used as generic struct.\r\nValues to not match their label.\r\nIe BaseColor here contains UVs."
   attributeGetTypes [
    #1
    #2
    #3
    #4
    #5
   ]
  }
  n2 = MaterialFunctionCall!(
   Texture Object=FunctionInput(Preview=TextureObject() {samplerType=SAMPLERTYPE_LinearColor, texture=/Engine/Functions/Engine_MaterialFunctions02/PivotPainter2/Black_1x1_EXR_Texture}) {inputName=ColorMap, inputType=FunctionInput_Texture2D, sortPriority=1}
   R UV=@n1.BaseColor
   G UV=@n1.BaseColor
   B UV=@n1.BaseColor
   Weights=@n1.BaseColor
   Derivative UVs=@n1.BaseColor
  ) {desc=Color Map, materialFunction=/ImpostorBaker/MaterialFunctions/Internal/FrameBlendWeights_TextureObject}
  FunctionOutput(None=MaterialFunctionCall!(Default=@n2.Result, Mobile Override=@n2.Result) {materialFunction=/ImpostorBaker/MaterialFunctions/Internal/OverrideMobile_V4}.Result) {description=BaseColor value of Impostor sampling., outputName=ColorMap}
 ]}
