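# Material function that renders an octahedral impostor from a single atlas frame, outputting BaseColor, Normal and WPO.
# The camera direction maps to hemi or full octahedral coordinates (static switch), selecting the atlas cell and the frame axes that orient the quad through WPO.
# A one-step parallax offset from a texture alpha and the frame-space vector shifts the UVs.
# Custom shadow-pass switches substitute the sky atmosphere light for the camera and drop parallax.
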MaterialFunction /ImpostorBaker/MaterialFunctions/Impostor_SingleFrameVersion
graph {nodes [
  n1 = VertexInterpolator(VS=NamedRerouteUsage())
  n2 = Add(A=Divide(A=Floor(None=@n1.PS), B=NamedRerouteUsage()), B=TextureCoordinate() {uTiling=10, vTiling=10})
  n3 = TextureSample(UVs=@n2, Tex=NamedRerouteUsage()) {texture=&Black}
  n4 = NamedRerouteUsage()
  n5 = NamedRerouteUsage()
  n6 = NamedRerouteUsage()
  n7 = NamedRerouteUsage()
  n8 = ComponentMask(None=VertexInterpolator(VS=MaterialFunctionCall!(VectorToTransform=@n4, BasisX=@n5, BasisY=@n6, BasisZ=@n7) {materialFunction=/Engine/Functions/Engine_MaterialFunctions02/Math/InverseTransformMatrix}.Transformed Vector 3).PS) {g=true, r=true}
  n9 = NamedRerouteUsage()
  MaterialExpressionMultiply_1 = Multiply(A=Multiply(A=Subtract(A=@n3.A) {constA=0.5, constB=0.5}, B=Divide(A=@n8, B=@n9) {constB=16}), B=NamedRerouteUsage())
  MaterialExpressionConstant_2 = Constant()
  n10 = Custom!(Main=@MaterialExpressionMultiply_1, Shadow=@MaterialExpressionConstant_2) {
   code="##if SHADOW_DEPTH_SHADER \r\n{\r\nreturn Shadow;\r\n}\r\n##endif\r\nreturn Main;"
   description=Shadow Pass Switch
   outputType=CMOT_Float2
   inputs [
    {inputName=Main, input {expression=&self:MaterialExpressionMultiply_1}}
    {inputName=Shadow, input {expression=&self:MaterialExpressionConstant_2}}
   ]
  }
  n11 = Add(A=@n2, B=@n10)
  n12 = TextureSample(UVs=@n11, Tex=NamedRerouteUsage()) {texture=&Black}
  FunctionOutput(None=@n12.RGBA) {outputName=BaseColor RGB}
  FunctionOutput(None=@n12.A) {outputName=BaseColor A, sortPriority=1}
  n13 = TextureSample(UVs=@n11, Tex=NamedRerouteUsage()) {texture=&Black}
  FunctionOutput(None=@n13.RGB) {outputName=Normal RGB, sortPriority=2}
  FunctionOutput(None=@n13.A) {outputName=Normal A, sortPriority=3}
  n14 = TextureCoordinate() {uTiling=10, vTiling=10}
  n15 = NamedRerouteUsage()
  n16 = MaterialFunctionCall(Float2=ConstantBiasScale(None=Multiply(A=@n14, B=@n15)) {bias=-0.5, scale=2}) {materialFunction=&BreakOutFloat2Components}
  n17 = Multiply(A=NamedRerouteUsage(), B=@n16.R)
  n18 = Multiply(A=MaterialFunctionCall() {materialFunction=/ImpostorBaker/MaterialFunctions/Internal/ImpostorInfo}.Scaled Size) {constB=0.5}
  n19 = Multiply(A=NamedRerouteUsage(), B=@n16.G)
  n20 = Transform(None=Add(A=Multiply(A=@n17, B=@n18), B=Multiply(A=@n19, B=@n18)))
  n21 = WorldPosition() {worldPositionShaderOffset=WPT_ExcludeAllShaderOffsets}
  FunctionOutput(None=Add(A=Subtract(A=@n20, B=@n21.XYZ), B=MaterialFunctionCall() {materialFunction=&ImpostorInfo}.Position) {constB=0}) {outputName=WPO, sortPriority=4}
  NamedRerouteDeclaration(None=FunctionInput(Preview=TextureObject() {samplerType=SAMPLERTYPE_LinearColor, texture=/Engine/Functions/Engine_MaterialFunctions02/PivotPainter2/Black_1x1_EXR_Texture}) {inputName=Normal, inputType=FunctionInput_Texture2D, sortPriority=1}) {name=NormalTex, nodeColor=<rgba 0.0937 0.4331 1 1>}
  n22 = NamedRerouteUsage()
  n23 = NamedRerouteUsage()
  n24 = ConstantBiasScale(None=Divide(A=Floor(None=@n22), B=Subtract(A=@n23))) {bias=-0.5, scale=2}
  n25 = StaticSwitchParameter(
   True=MaterialFunctionCall!(Hemi Octahedron=@n24) {materialFunction=/ImpostorBaker/MaterialFunctions/Internal/HemiOctahedronToUnitVector}.Unit Vector
   False=MaterialFunctionCall!(Octahedron=@n24) {materialFunction=/Engine/Functions/Engine_MaterialFunctions01/Vectors/OctahedronToUnitVector}.Unit Vector
  ) {
   defaultValue=true
   desc="Defaults to True. When True, the impostor is only captured to support views from the upper hemisphere. Otherwise it will include views from all views. Restricting to the upper hemisphere is useful for trees that are unlikely to be seen from below when far away."
   parameterName=UpperHemisphereOnlyImpostor
  }
  n26 = NamedRerouteDeclaration(None=Normalize(VectorInput=@n25) {desc=Z}) {name=Frame Z, nodeColor=<rgba 0.7344 1 0 1>}
  n27 = MaterialFunctionCall(Float3=@n26) {materialFunction=&BreakOutFloat3Components}
  n28 = Multiply(A=@n27.R) {constB=-1}
  n29 = Constant()
  n30 = NamedRerouteDeclaration(None=Normalize(VectorInput=MaterialFunctionCall(X=@n27.G, Y=@n28, Z=@n29) {materialFunction=&MakeFloat3}.Result) {desc=X}) {name=Frame X, nodeColor=<rgba 0 0.5781 1 1>}
  NamedRerouteDeclaration(None=Normalize(VectorInput=CrossProduct(A=@n26, B=@n30)) {desc=Y}) {name=Frame Y, nodeColor=<rgba 0 0.7891 1 1>}
  MaterialExpressionCameraPositionWS_6 = CameraPositionWS()
  MaterialExpressionMultiply_29 = Multiply(A=SkyAtmosphereLightDirection()) {constB=500000}
  n31 = Custom(Main=@MaterialExpressionCameraPositionWS_6, Shadow=@MaterialExpressionMultiply_29) {
   code="##if SHADOW_DEPTH_SHADER \r\n{\r\nreturn Shadow;\r\n}\r\n##endif\r\nreturn Main;"
   description=Shadow Pass Switch
   inputs [
    {inputName=Main, input {expression=&self:MaterialExpressionCameraPositionWS_6}}
    {inputName=Shadow, input {expression=&self:MaterialExpressionMultiply_29}}
   ]
  }
  n32 = Subtract(A=@n31, B=MaterialFunctionCall() {materialFunction=&ImpostorInfo}.Position)
  NamedRerouteDeclaration(None=Transform(None=Normalize(VectorInput=@n32)) {transformSourceType=TRANSFORMSOURCE_World, transformType=TRANSFORM_Tangent}) {name=Ray Origin Local, nodeColor=<rgba 0.0781 0 1 1>}
  NamedRerouteDeclaration(None=FunctionInput(Preview=TextureObject() {samplerType=SAMPLERTYPE_LinearColor, texture=&Black_1x1_EXR_Texture}) {inputName=BaseColor, inputType=FunctionInput_Texture2D}) {name=BaseColorTex, nodeColor=<rgba 0.0937 0.4331 1 1>}
  n33 = MaterialFunctionCall(Float3=NamedRerouteUsage()) {materialFunction=&BreakOutFloat3Components}
  n34 = Normalize(VectorInput=MaterialFunctionCall(X=@n33.R, Y=@n33.R, Z=Max(A=@n33.B) {constB=0.001}) {materialFunction=&MakeFloat3}.Result)
  n35 = StaticSwitchParameter(
   True=MaterialFunctionCall!(Unit Vector=@n34) {materialFunction=/ImpostorBaker/MaterialFunctions/Internal/UnitVectorToHemiOctahedra}.Hemi Octahedron
   False=MaterialFunctionCall!(Vector=@n34) {materialFunction=/Engine/Functions/Engine_MaterialFunctions01/Vectors/UnitVectorToOctahedron}.Octahedron
  ) {
   defaultValue=true
   desc="Defaults to True. When True, the impostor is only captured to support views from the upper hemisphere. Otherwise it will include views from all views. Restricting to the upper hemisphere is useful for trees that are unlikely to be seen from below when far away."
   parameterName=UpperHemisphereOnlyImpostor
  }
  n36 = Saturate(None=ConstantBiasScale(None=@n35))
  n37 = FunctionInput() {bUsePreviewValueAsDefault=true, inputName=Frames XY, inputType=FunctionInput_Scalar, previewValue=<16 0 0 1>}
  n38 = Subtract(A=NamedRerouteDeclaration(None=@n37) {name=Frames XY, nodeColor=<rgba 1 0 0.4922 1>})
  NamedRerouteDeclaration(None=Add(A=Multiply(A=@n36, B=@n38)) {constB=0.5}) {name=Octahedron Scaled, nodeColor=<rgba 1 0.3281 0 1>}
  NamedRerouteDeclaration(None=StaticSwitchParameter(True=ScalarParameter() {defaultValue=1, parameterName=EnableParallax}, False=Constant() {r=1}) {desc=Useful to Toggle Parallax without recompiling, parameterName=Parallax Debug Toggle}) {name=Enable Parallax, nodeColor=<rgba 0 0.4375 1 1>}
 ]}
