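# Internal ImpostorBaker material function for octahedral impostors: outputs three per-frame UVs, blend weights, sprite WPO, depth-derived PDO and debug grid data.
# The view ray selects three neighboring frames, projects onto their virtual planes, then Impostor_Find_Intersection refines clamped UVs and depths.
# Depths drive PDO and optional depth-derived weights.
# Shadow passes take the ray from the sky light direction; static switches control vertex interpolation, per-pixel frame selection and debugging.
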MaterialFunction /ImpostorBaker/MaterialFunctions/Internal/Impostor_UVs_and_WPO
graph {nodes [
  Composite() {subgraphName=Debug Impostor Grid}
  Composite() {subgraphName=Derive Accurate Weights and PDO}
  n1 = NamedRerouteUsage()
  n2 = MaterialFunctionCall!(
   NormalTex=NamedRerouteUsage()
   ScalarMaps=NamedRerouteUsage()
   UVs=Add(A=NamedRerouteUsage() {desc="0,0"}, B=NamedRerouteUsage())
   RayDir=NamedRerouteUsage()
   FramesXY=@n1
  ) {materialFunction=&/Impostor_Find_Intersection}
  FunctionOutput(None=MaterialFunctionCall!(
    UV=NamedRerouteDeclaration(None=@n2.UV) {name=Final UV R, nodeColor=<rgba 0.5 0 0 1>}
    Frame Index=NamedRerouteUsage() {desc="0,0"}
    FramesXY=@n1
   ) {materialFunction=&/ClampImpostorFrameUVs}.Result) {outputName=UV Plane R}
  n3 = Add(
   A=NamedRerouteUsage() {desc="1,0 or 0,1"}
   B=LinearInterpolate(A=Constant2Vector() {g=1}, B=Constant2Vector() {r=1}, Alpha=NamedRerouteUsage())
  )
  n4 = NamedRerouteUsage()
  n5 = MaterialFunctionCall(
   NormalTex=NamedRerouteUsage()
   ScalarMaps=NamedRerouteUsage()
   UVs=Add(A=NamedRerouteDeclaration(None=@n3) {name=Frame Index +X or +Y, nodeColor=<rgba 0.5 0 1 1>}, B=NamedRerouteUsage())
   RayDir=NamedRerouteUsage()
   FramesXY=@n4
  ) {materialFunction=&Impostor_Find_Intersection}
  FunctionOutput(None=MaterialFunctionCall(UV=NamedRerouteDeclaration(None=@n5.UV) {name=Final UV G, nodeColor=<rgba 0 0.5 0 1>}, Frame Index=NamedRerouteUsage(), FramesXY=@n4) {materialFunction=&ClampImpostorFrameUVs}.Result) {outputName=UV Plane G, sortPriority=1}
  n6 = Add(A=NamedRerouteUsage() {desc="1,1"}, B=Constant2Vector() {g=1, r=1})
  n7 = NamedRerouteUsage()
  n8 = MaterialFunctionCall(
   NormalTex=NamedRerouteUsage()
   ScalarMaps=NamedRerouteUsage()
   UVs=Add(A=NamedRerouteDeclaration(None=@n6) {name=Frame Index +XY, nodeColor=<rgba 0 0.5312 1 1>}, B=NamedRerouteUsage())
   RayDir=NamedRerouteUsage()
   FramesXY=@n7
  ) {materialFunction=&Impostor_Find_Intersection}
  FunctionOutput(None=MaterialFunctionCall(UV=NamedRerouteDeclaration(None=@n8.UV) {name=Final UV B, nodeColor=<rgba 0 0 0.5 1>}, Frame Index=NamedRerouteUsage(), FramesXY=@n7) {materialFunction=&ClampImpostorFrameUVs}.Result) {outputName=UV Plane B, sortPriority=2}
  n9 = Multiply(A=NamedRerouteUsage(), B=NamedRerouteUsage())
  n10 = NamedRerouteUsage()
  n11 = Subtract(A=AppendVector(A=Frac(None=@n9), B=OneMinus(None=@n10))) {constB=0.5}
  n12 = NamedRerouteUsage()
  n13 = NamedRerouteUsage()
  n14 = Multiply(A=NamedRerouteUsage()) {constB=-1}
  n15 = MaterialFunctionCall() {materialFunction=&/ImpostorInfo}
  n16 = Add(
   A=Multiply(
    A=MaterialFunctionCall(VectorToTransform=@n11, BasisX=@n12, BasisY=@n13, BasisZ=@n14) {materialFunction=&Transform3x3Matrix}.Transformed Vector
    B=@n15.Scaled Size
   )
   B=@n15.Position
  )
  n17 = Multiply(A=NamedRerouteUsage(), B=NamedRerouteUsage())
  n18 = NamedRerouteUsage()
  n19 = Subtract(A=AppendVector(A=Frac(None=@n17), B=OneMinus(None=@n18))) {constB=0.5}
  n20 = NamedRerouteUsage()
  n21 = NamedRerouteUsage()
  n22 = Multiply(A=NamedRerouteUsage()) {constB=-1}
  n23 = MaterialFunctionCall() {materialFunction=&ImpostorInfo}
  n24 = Add(
   A=Multiply(
    A=MaterialFunctionCall(VectorToTransform=@n19, BasisX=@n20, BasisY=@n21, BasisZ=@n22) {materialFunction=&Transform3x3Matrix}.Transformed Vector
    B=@n23.Scaled Size
   )
   B=@n23.Position
  )
  n25 = Multiply(A=NamedRerouteUsage(), B=NamedRerouteUsage())
  n26 = NamedRerouteUsage()
  n27 = Subtract(A=AppendVector(A=Frac(None=@n25), B=OneMinus(None=@n26))) {constB=0.5}
  n28 = NamedRerouteUsage()
  n29 = NamedRerouteUsage()
  n30 = Multiply(A=NamedRerouteUsage()) {constB=-1}
  n31 = MaterialFunctionCall() {materialFunction=&ImpostorInfo}
  n32 = Add(
   A=Multiply(
    A=MaterialFunctionCall(VectorToTransform=@n27, BasisX=@n28, BasisY=@n29, BasisZ=@n30) {materialFunction=&Transform3x3Matrix}.Transformed Vector
    B=@n31.Scaled Size
   )
   B=@n31.Position
  )
  n33 = MaterialFunctionCall(X=Distance(A=@n16, B=WorldPosition().XYZ), Y=Distance(A=@n24, B=WorldPosition().XYZ), Z=Distance(A=@n32, B=WorldPosition().XYZ)) {materialFunction=&MakeFloat3}
  MaterialExpressionReroute_4 = Reroute(None=@n33.Result)
  n34 = MaterialFunctionCall(Float3=NamedRerouteDeclaration(None=@MaterialExpressionReroute_4.Depth Offsets) {name=CorrectedDepths, nodeColor=<rgba 1 0 0.9375 1>}) {materialFunction=&BreakOutFloat3Components}
  n35 = ComponentMask(None=MaterialFunctionCall!(R=@n34.R, G=@n34.R, B=@n34.R, Weights=NamedRerouteUsage()) {desc=Interpolated Depth, materialFunction=&/FrameBlendWeights}.Result) {r=true}
  n36 = CameraVectorWS()
  n37 = Abs(None=DotProduct(
    A=Normalize(VectorInput=@n36)
    B=MaterialFunctionCall() {materialFunction=&CameraDirectionVector}.Result
   ))
  FunctionOutput(None=NamedRerouteDeclaration(None=Multiply(A=@n35, B=@n37) {constB=0.5}) {name=Accurate PDO, nodeColor=<rgba 0 0.9766 1 1>}) {outputName=Depth Derived PDO, sortPriority=15}
  n38 = FunctionInput() {bUsePreviewValueAsDefault=true, inputName=Frames XY, inputType=FunctionInput_Scalar, previewValue=<16 0 0 1>}
  n39 = MaterialFunctionCall() {materialFunction=&ImpostorInfo}
  n40 = MaterialFunctionCall!(
   Center Location=@n39.Position
   0-1 UVs=Multiply(A=TextureCoordinate() {uTiling=10, vTiling=10}, B=NamedRerouteDeclaration(None=@n38) {name=Frames XY, nodeColor=<rgba 1 0 0.4922 1>})
   XY scale=Multiply(A=@n39.Scaled Size, B=NamedRerouteUsage())
   Ray Origin=NamedRerouteUsage()
   "Normals (Optional)"
  ) {materialFunction=&/Sprite_PositionBased}
  n41 = NamedRerouteUsage()
  n42 = DistanceFieldsRenderingSwitch!(No=Constant(), Yes=@n41)
  n43 = ShadowReplace(Default=FeatureLevelSwitch(Default=@n41, ES2=@n42, ES3_1=@n42, SM4_REMOVED=@n41, SM5=@n41, SM6=@n41), Shadow=Constant())
  FunctionOutput(None=NamedRerouteDeclaration(None=Add(A=@n40.World position offset, B=@n43) {constB=0}) {name=WorldPositionOffset, nodeColor=<rgba 1 0.0234 0 1>}) {outputName=WPO, sortPriority=7}
  n44 = Multiply(A=NamedRerouteUsage(), B=Subtract(A=NamedRerouteUsage()))
  n45 = Multiply(A=NamedRerouteUsage(), B=Subtract(A=NamedRerouteUsage()))
  n46 = Multiply(
   A=SphereMask(A=Floor(None=@n44), B=Floor(None=@n45), Radius=ScalarParameter() {defaultValue=0.01, parameterName=FrameDebugRadius}) {desc=Isolate Current Frame}
   B=NamedRerouteUsage()
  )
  n47 = NamedRerouteUsage()
  n48 = NamedRerouteUsage()
  n49 = ScalarParameter() {defaultValue=0.015, parameterName=FrameDebugRadius2}
  n50 = MaterialFunctionCall(UVs=Multiply(A=TextureCoordinate() {uTiling=10, vTiling=10}, B=NamedRerouteUsage()), Texture Scale=Divide(B=NamedRerouteUsage())) {materialFunction=&ScaleUVsByCenter}
  n51 = NamedRerouteUsage()
  n52 = Frac(None=Multiply(A=NamedRerouteDeclaration(None=@n50.UVs) {name=NormalizedDebugUV, nodeColor=<rgba 1 0 0.5859 1>}, B=Subtract(A=@n51)))
  n53 = ScalarParameter() {defaultValue=0.03, parameterName=grid line}
  n54 = Add(
   A=Add(A=Multiply(A=@n46) {constB=0.5}, B=SphereMask(A=@n47, B=@n48, Radius=@n49) {desc=Draw Dot for Ray})
   B=Multiply(
    A=ScalarParameter() {defaultValue=1, parameterName=Debug Grid}
    B=MaterialFunctionCall!(UV=@n52, Line Width=@n53) {materialFunction=&/TriangulatedWireframe}.Result
   )
  )
  MaterialExpressionReroute_1 = Reroute(None=@n54)
  n55 = Multiply(A=@MaterialExpressionReroute_1.Output, B=ScalarParameter() {defaultValue=1, parameterName=ShowDebug})
  n56 = Constant()
  FunctionOutput(None=NamedRerouteDeclaration(None=StaticSwitchParameter(True=@n55, False=@n56) {parameterName=Show Frame Debug Info}) {name=Frame Debug Display, nodeColor=<rgba 0 0 0 1>}) {outputName=Debug Color, sortPriority=8}
  n57 = DotProduct(A=@MaterialExpressionReroute_1.Output, B=Constant() {r=2})
  n58 = ScalarParameter() {defaultValue=1, parameterName=ShowDebug}
  n59 = MaterialFunctionCall(Float2=NamedRerouteUsage()) {materialFunction=&BreakOutFloat2Components}
  n60 = Constant() {r=0.5}
  n61 = Multiply(A=SphereMask(A=@n59.R, B=@n60) {attenuationRadius=0.5}, B=SphereMask(A=@n59.G, B=@n60) {attenuationRadius=0.5})
  n62 = StaticSwitchParameter(True=Multiply(A=Multiply(A=@n57, B=@n58), B=@n61), False=Constant()) {parameterName=Show Frame Debug Info}
  n63 = StaticSwitchParameter(True=@n61, False=Constant() {r=1}) {parameterName=Show Frame Debug Info}
  FunctionOutput(None=AppendVector(
    A=NamedRerouteDeclaration(None=@n62) {name=Frame DebugMask, nodeColor=<rgba 0 0 0 1>}
    B=NamedRerouteDeclaration(None=@n63) {name=Debug Box Clip, nodeColor=<rgba 0 0 0 1>}
   )) {outputName=Debug Mask + Clip, sortPriority=9}
  n64 = NamedRerouteUsage()
  n65 = ScalarParameter() {parameterName=ManualIndex}
  n66 = LinearInterpolate(
   A=LinearInterpolate(A=Constant3Vector() {constant=<rgba 1 0 0 0>}, B=Constant3Vector() {constant=<rgba 0 1 0 0>}, Alpha=Saturate(None=@n65))
   B=Constant3Vector() {constant=<rgba 0 0 1 0>}
   Alpha=Saturate(None=Subtract(A=@n65))
  )
  n67 = ScalarParameter() {defaultValue=1, parameterName=Manual Weights}
  FunctionOutput(None=StaticSwitchParameter(True=LinearInterpolate(A=@n64, B=@n66, Alpha=@n67), False=NamedRerouteUsage()) {parameterName=Enable Manual Weights}) {outputName=Blend Weights, sortPriority=6}
  n68 = MaterialFunctionCall!(
   InVec=NamedRerouteUsage()
   X Vec
   Y Vec
   Z Vec=MaterialFunctionCall!(Frame=NamedRerouteUsage(), XY Frames=NamedRerouteUsage()) {materialFunction=&/ImposterFrameTransform_Setup}.World Z
  ) {materialFunction=&/ImposterFrameTransform}
  n69 = MaterialFunctionCall!(
   Plane X Axis=NamedRerouteDeclaration(None=@n68.World X) {name=PlaneR_X, nodeColor=<rgba 1 0.0263 0 1>}
   Plane Y Axis=NamedRerouteDeclaration(None=@n68.World Y) {name=PlaneR_Y, nodeColor=<rgba 0.0245 1 0 1>}
   Plane Normal Axis=NamedRerouteDeclaration(None=@n68.World Z) {name=PlaneR_Z, nodeColor=<rgba 0 0.2848 1 1>}
   UV Scale=NamedRerouteUsage()
   Ray Dir Local=NamedRerouteUsage()
   Ray Origin Local=NamedRerouteUsage()
  ) {materialFunction=&/VirtualPlaneCoordinates_Imposter}
  n70 = StaticSwitchParameter(True=VertexInterpolator(VS=@n69.UVs).PS, False=@n69.UVs) {defaultValue=true, parameterName=Use Vertex Interpolation}
  NamedRerouteDeclaration(None=@n70) {name=Plane R UV, nodeColor=<rgba 1 0.0469 0 1>}
  n71 = NamedRerouteUsage()
  n72 = LinearInterpolate(A=Constant2Vector() {g=1}, B=Constant2Vector() {r=1}, Alpha=NamedRerouteUsage())
  n73 = MaterialFunctionCall(
   InVec=NamedRerouteUsage()
   Z Vec=MaterialFunctionCall(Frame=Add(A=@n71, B=@n72), XY Frames=NamedRerouteUsage()) {materialFunction=&ImposterFrameTransform_Setup}.World Z
  ) {materialFunction=&ImposterFrameTransform}
  n74 = MaterialFunctionCall(
   Plane X Axis=NamedRerouteDeclaration(None=@n73.World X) {name=PlaneG_X, nodeColor=<rgba 1 0.0263 0 1>}
   Plane Y Axis=NamedRerouteDeclaration(None=@n73.World Y) {name=PlaneG_Y, nodeColor=<rgba 0.0245 1 0 1>}
   Plane Normal Axis=NamedRerouteDeclaration(None=@n73.World Z) {name=PlaneG_Z, nodeColor=<rgba 0 0.2848 1 1>}
   UV Scale=NamedRerouteUsage()
   Ray Dir Local=NamedRerouteUsage()
   Ray Origin Local=NamedRerouteUsage()
  ) {materialFunction=&VirtualPlaneCoordinates_Imposter}
  n75 = StaticSwitchParameter(True=VertexInterpolator(VS=@n74.UVs).PS, False=@n74.UVs) {defaultValue=true, parameterName=Use Vertex Interpolation}
  NamedRerouteDeclaration(None=@n75) {name=Plane G UV, nodeColor=<rgba 0 1 0 1>}
  n76 = NamedRerouteUsage()
  n77 = Constant2Vector() {g=1, r=1}
  n78 = MaterialFunctionCall(
   InVec=NamedRerouteUsage()
   Z Vec=MaterialFunctionCall(Frame=Add(A=@n76, B=@n77), XY Frames=NamedRerouteUsage()) {materialFunction=&ImposterFrameTransform_Setup}.World Z
  ) {materialFunction=&ImposterFrameTransform}
  n79 = MaterialFunctionCall(
   Plane X Axis=NamedRerouteDeclaration(None=@n78.World X) {name=PlaneB_X, nodeColor=<rgba 1 0.0263 0 1>}
   Plane Y Axis=NamedRerouteDeclaration(None=@n78.World Y) {name=PlaneB_Y, nodeColor=<rgba 0.0245 1 0 1>}
   Plane Normal Axis=NamedRerouteDeclaration(None=@n78.World Z) {name=PlaneB_Z, nodeColor=<rgba 0 0.2848 1 1>}
   UV Scale=NamedRerouteUsage()
   Ray Dir Local=NamedRerouteUsage()
   Ray Origin Local=NamedRerouteUsage()
  ) {materialFunction=&VirtualPlaneCoordinates_Imposter}
  n80 = StaticSwitchParameter(True=VertexInterpolator(VS=@n79.UVs).PS, False=@n79.UVs) {defaultValue=true, parameterName=Use Vertex Interpolation}
  NamedRerouteDeclaration(None=@n80) {name=Plane B UV, nodeColor=<rgba 0 0 1 1>}
  NamedRerouteDeclaration(None=FunctionInput(Preview=TextureObject() {samplerType=SAMPLERTYPE_LinearColor, texture=/Engine/Functions/Engine_MaterialFunctions02/PivotPainter2/Black_1x1_EXR_Texture}) {inputName=Normal, inputType=FunctionInput_Texture2D, sortPriority=2}) {name=NormalTex, nodeColor=<rgba 0.0937 0.4331 1 1>}
  n81 = MaterialFunctionCall() {materialFunction=&ImpostorInfo}
  NamedRerouteDeclaration(None=@n81.Scaled Size) {name=Scaled Size, nodeColor=<rgba 0 1 0.9062 1>}
  n82 = NamedRerouteUsage()
  NamedRerouteDeclaration(None=ShadowReplace(Default=Floor(None=@n82), Shadow=Round(None=@n82))) {name=Frame Index PS, nodeColor=<rgba 0 0.1609 1 1>}
  n83 = MaterialFunctionCall!(In=@n68.Local Z Vector) {materialFunction=&/SeparateRG-B}
  n84 = VertexInterpolator(VS=@n83.RG)
  NamedRerouteDeclaration(None=AppendVector(A=StaticSwitchParameter(True=@n84.PS, False=@n83.RG) {defaultValue=true, parameterName=Use Vertex Interpolation}, B=@n83.B)) {name=Ray Dir Plane R, nodeColor=<rgba 0.125 0 1 1>}
  n85 = MaterialFunctionCall() {materialFunction=&ImpostorInfo}
  n86 = Divide(A=Subtract(A=NamedRerouteUsage(), B=NamedRerouteUsage()), B=Distance(A=NamedRerouteUsage(), B=@n85.Position))
  n87 = Multiply(A=@n85.Scaled Size) {constB=0.5}
  n88 = ScalarParameter() {
   defaultValue=1
   desc=Pushes the sprite to the front of the object which allows pixel depth offset to work properly.
   group=Advanced
   parameterName=Camera Push
  }
  NamedRerouteDeclaration(None=Multiply(A=Multiply(A=@n86, B=@n87), B=@n88)) {name=Camera Plane Alignment, nodeColor=<rgba 1 0.375 0 1>}
  n89 = ScalarParameter() {
   defaultValue=1
   desc="In rare cases, parallax might cause the object to leave the bounds of the geometry as if clipped. WPOScale can be increased above 1 to scale up the sprite to fix this. Be sure to use the smallest increase necessary as there will be a performance penalty."
   group=Advanced
   parameterName=WPOScale
  }
  NamedRerouteDeclaration(None=StaticSwitchParameter(True=Multiply(A=@n89) {constB=1.5}, False=@n89) {parameterName=Show Frame Debug Info}) {name=WPO Scale, nodeColor=<rgba 0.0634 0.7524 0.8646 1>}
  NamedRerouteDeclaration(None=@n40.World Position) {name=WorldPos+Sprite, nodeColor=<rgba 0 0 0 1>}
  n90 = NamedRerouteUsage()
  n91 = StaticBoolParameter() {
   defaultValue=true
   desc="Defaults to True. When True, the impostor is only captured to support views from the upper hemisphere. Otherwise it will include views from all views. \r\nRestricting to the upper hemisphere is useful for trees that are unlikely to be seen from below when far away."
   parameterName=UpperHemisphereOnlyImpostor
  }
  n92 = Saturate(None=ConstantBiasScale(None=MaterialFunctionCall!(2D Grid Value, Vector=@n90, Upper Hemisphere Only=@n91) {materialFunction=&/ImpostorGridToVector}.2D Grid Value))
  n93 = NamedRerouteUsage()
  NamedRerouteDeclaration(None=Multiply(A=NamedRerouteDeclaration(None=@n92) {name=Octahedron, nodeColor=<rgba 0.1019 0.4322 1 1>}, B=Subtract(A=@n93))) {name=Octahedron Scaled, nodeColor=<rgba 0.1014 0.326 1 1>}
  n94 = MaterialFunctionCall(In=@n73.Local Z Vector) {materialFunction=&SeparateRG-B}
  n95 = VertexInterpolator(VS=@n94.RG)
  NamedRerouteDeclaration(None=AppendVector(A=StaticSwitchParameter(True=@n95.PS, False=@n94.RG) {defaultValue=true, parameterName=Use Vertex Interpolation}, B=@n94.B)) {name=Ray Dir Plane G, nodeColor=<rgba 0 1 0 1>}
  n96 = Multiply(A=SkyAtmosphereLightDirection()) {constB=50000}
  n97 = NamedRerouteDeclaration(None=ShadowReplace(Default=CameraPositionWS(), Shadow=Add(A=@n96, B=@n81.Position))) {name=Ray Origin, nodeColor=<rgba 0.3359 1 0 1>}
  n98 = NamedRerouteUsage()
  n99 = Normalize(VectorInput=Transform(None=Subtract(A=@n97, B=@n98)) {transformSourceType=TRANSFORMSOURCE_World, transformType=TRANSFORM_Tangent})
  n100 = MaterialFunctionCall() {materialFunction=&ImpostorInfo}
  n101 = NamedRerouteDeclaration(None=Transform(None=Subtract(A=@n97, B=@n100.Position)) {transformSourceType=TRANSFORMSOURCE_World, transformType=TRANSFORM_Tangent}) {name=Ray Origin Local, nodeColor=<rgba 0.2891 1 0 1>}
  n102 = StaticSwitchParameter(
   True=NamedRerouteDeclaration(None=@n99) {name=RayDir Perspective, nodeColor=<rgba 0 0 0 1>}
   False=NamedRerouteDeclaration(None=Normalize(VectorInput=@n101)) {name=RayDir Ortho, nodeColor=<rgba 0 0 0 1>}
  ) {
   defaultValue=true
   desc="Defaults to True.  When False, a single View Ray will be used to the center\r\nfor the entire objet. When True, the ray will be evaluated at each vertex."
   parameterName=Interpolate Ray Position
  }
  NamedRerouteDeclaration(None=@n102) {name=RayDir, nodeColor=<rgba 0.0853 0.3677 0.6302 1>}
  n103 = NamedRerouteUsage()
  n104 = StaticSwitchParameter(True=VertexInterpolator(VS=@n103).PS, False=@n103) {defaultValue=true, parameterName=Use Vertex Interpolation}
  n105 = MaterialFunctionCall!(UV=Frac(None=@n104)) {materialFunction=&/TriangleInterpolator}
  NamedRerouteDeclaration(None=@n105.Mask) {name=Triangle Mask VS, nodeColor=<rgba 1 0.3802 0 1>}
  NamedRerouteDeclaration(None=MaterialFunctionCall(UV=Frac(None=@n82)) {materialFunction=&TriangleInterpolator}.Mask) {name=Triangle Mask PS, nodeColor=<rgba 1 0.4453 0 1>}
  n106 = MaterialFunctionCall(In=@n78.Local Z Vector) {materialFunction=&SeparateRG-B}
  n107 = VertexInterpolator(VS=@n106.RG)
  NamedRerouteDeclaration(None=AppendVector(A=StaticSwitchParameter(True=@n107.PS, False=@n106.RG) {defaultValue=true, parameterName=Use Vertex Interpolation}, B=@n106.B)) {name=Ray Dir Plane B, nodeColor=<rgba 0 0 1 1>}
  NamedRerouteDeclaration(None=@n2.Depth) {name=Depth R, nodeColor=<rgba 0.9219 0 1 1>}
  NamedRerouteDeclaration(None=@n5.Depth) {name=Depth G, nodeColor=<rgba 0 1 0 1>}
  n108 = MaterialFunctionCall(UV=Frac(None=@n45)) {materialFunction=&TriangleInterpolator}
  n109 = NamedRerouteUsage()
  n110 = Subtract(A=NamedRerouteUsage())
  n111 = MaterialFunctionCall(UV=Frac(None=Multiply(A=@n109, B=@n110))) {materialFunction=&TriangleInterpolator}
  n112 = LinearInterpolate(A=OneMinus(None=@n108.Mask), B=@n108.Mask, Alpha=@n111.Mask)
  n113 = Constant() {r=1}
  n114 = MaterialFunctionCall!(UVs=NamedRerouteUsage(), Tiling=ScalarParameter() {defaultValue=384, parameterName=DebugShadowCheckerDither}) {materialFunction=/Engine/ArtTools/RenderToTexture/MaterialFunctions/CheckerPattern}
  NamedRerouteDeclaration(None=Multiply(A=Multiply(A=@n112, B=@n111.Result), B=ShadowReplace(Default=@n113, Shadow=@n114.Result))) {name=Debug Triangle Mask, nodeColor=<rgba 1 0 0.375 1>}
  NamedRerouteDeclaration(None=FunctionInput(Preview=TextureObject() {samplerType=SAMPLERTYPE_LinearColor, texture=&Black_1x1_EXR_Texture}) {inputName=BaseColor, inputType=FunctionInput_Texture2D, sortPriority=1}) {name=BaseColorTex, nodeColor=<rgba 0.0937 0.4331 1 1>}
  NamedRerouteDeclaration(None=ShadowReplace(Default=Floor(None=@n104), Shadow=Round(None=@n104))) {name=Frame Index VS, nodeColor=<rgba 0 0 0 1>}
  NamedRerouteDeclaration(None=@n8.Depth) {name=Depth B, nodeColor=<rgba 0 0 1 1>}
  n115 = MaterialFunctionCall(Float3=NamedRerouteUsage()) {materialFunction=&BreakOutFloat3Components}
  n116 = Max(A=@n115.R, B=@n115.R)
  n117 = ScalarParameter() {parameterName=Depth Modulation Bias}
  n118 = Subtract(A=@n115.G, B=Max(A=@n115.R, B=@n115.R))
  n119 = Subtract(A=Subtract(A=@n115.R, B=Max(A=@n115.G, B=@n115.G)), B=@n117)
  n120 = Constant3Vector() {constant=<rgba 1 0 0 0>}
  n121 = NamedRerouteUsage()
  n122 = If(
   A=Subtract(A=Subtract(A=@n115.B, B=@n116), B=@n117)
   "A > B"=Constant3Vector() {constant=<rgba 0 0 1 0>}
   "A < B"=If(A=Subtract(A=@n118, B=@n117), "A > B"=Constant3Vector() {constant=<rgba 0 1 0 0>}, "A < B"=If(A=@n119, "A > B"=@n120, "A < B"=@n121))
  )
  n123 = NamedRerouteUsage()
  NamedRerouteDeclaration(None=StaticSwitchParameter(True=StaticSwitchParameter(True=@n122, False=@n123) {defaultValue=true, parameterName=Depth Derived Weights}, False=@n123) {parameterName=Multi Iteration Parallax}) {name=Blend Weights, nodeColor=<rgba 1 0 0.6328 1>}
  NamedRerouteDeclaration(None=@n105.Result) {name=Triangle Weights VS, nodeColor=<rgba 0.0593 0.4665 0.7656 1>}
  n124 = FunctionInput(Preview=TextureObject() {samplerType=SAMPLERTYPE_LinearColor, texture=&Black_1x1_EXR_Texture}) {
   description=" Could be used for things like Specular, Roughness, OpacityMask etc."
   inputName=ScalarMaps
   inputType=FunctionInput_Texture2D
   sortPriority=4
  }
  NamedRerouteDeclaration(None=@n124) {name=ScalarMapsTex, nodeColor=<rgba 0.0937 0.4331 1 1>}
  n125 = StaticSwitchParameter(
   True=StaticSwitchParameter(True=NamedRerouteUsage(), False=NamedRerouteUsage()) {defaultValue=true, parameterName=Use Vertex Interpolation}
   False=NamedRerouteUsage()
  ) {
   desc="Defaults to False.  When True, the View Ray will per pixel, allowing different pixels\r\nto choose from different frames. Only works when Vertex Interpolation is disabled."
   parameterName=Per Pixel Frames
  }
  NamedRerouteDeclaration(None=@n125) {name=RayDir_FrameSelection, nodeColor=<rgba 0.0853 0.3677 0.6302 1>}
  n126 = MaterialFunctionCall(X=Distance(A=@n16, B=CameraPositionWS()), Y=Distance(A=@n24, B=CameraPositionWS()), Z=Distance(A=@n32, B=CameraPositionWS())) {materialFunction=&MakeFloat3}
  MaterialExpressionReroute_6 = Reroute(None=@n126.Result)
  NamedRerouteDeclaration(None=@MaterialExpressionReroute_6.Distance from Camera) {name=Distance To Surface, nodeColor=<rgba 1 0 0.4219 1>}
  NamedRerouteDeclaration(None=Multiply(A=@n99) {constB=-1}) {name=To Plane Local, nodeColor=<rgba 0 0.1094 1 1>}
  NamedRerouteUsage()
  Normalize(VectorInput=NamedRerouteUsage())
  PinBase()
  PinBase!(Output=@n54) {reroutePins [Output {expression=&self:MaterialExpressionReroute_1}]}
  PinBase!()
  PinBase!(Depth Offsets=@n33.Result, Distance from Camera=@n126.Result) {reroutePins [
    Depth Offsets {expression=&self:MaterialExpressionReroute_4}
    Distance from Camera {expression=&self:MaterialExpressionReroute_6}
   ]}
  Saturate(None=@n80)
  Saturate(None=@n70)
  Saturate(None=@n75)
 ]}
